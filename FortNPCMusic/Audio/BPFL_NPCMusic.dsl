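; BPFL_NPCMusic is a Blueprint function library that provides a logging helper for the FortNPCMusic feature.
; LogNPCMusic prefixes the given text with "FortNPCMusic:" through a localized format string, converts it to a string and sends it to FortShippingLog.
; A boolean parameter is passed on to choose whether the message logs as a warning.
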(fn LogNPCMusic
  (params (Text :type Text) (bLogAsWarning :type Boolean))
  (Fortnite|FortShippingLog 0 (Utilities|String|ToString(Text) (Utilities|Text|FormatText "NSLOCTEXT(\"[B7D869B6CFC8DC9BE21C7C8CF7D95F35]\", \"EC5025AC4CCA143F2E46FF8BF07F59F7\", \"FortNPCMusic: {log}\")" Text)) bLogAsWarning))
