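# Template stereo MetaSound source for NPC music, with unaware and threatened layers and temporary jungle boss waves as defaults.
# A BPM-driven trigger clock quantizes music start, threat entry and stop, with intro stingers leading into the loops.
# The interpolated Threatened value crossfades both layers, and distance scales overall volume.
# Stopping halts the loops, then either plays a death stinger or releases an envelope, and signals OnFinished when done.
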MetaSoundSource /FortNPCMusic/Audio/MSS_NPCMusic_Template
inputs [
 NPC.Event.OnMusicStop {type=Trigger}
 NPC.Event.OnPlayerTargeted {type=Trigger}
 NPC.Event.OnThreatened {type=Trigger}
 NPC.Music.BPM {type=Float}
 NPC.Music.CrossfadeTime {type=Time, access=Constructor}
 NPC.Music.StartThreatenStingerWave {type=WaveAsset, access=Constructor}
 NPC.Music.StartUnawareStingerWave {type=WaveAsset, access=Constructor}
 NPC.Music.StopStingerWave {type=WaveAsset, access=Constructor}
 NPC.Music.ThreatenWave {type=WaveAsset, access=Constructor}
 NPC.Music.UnawareWave {type=WaveAsset, access=Constructor}
 NPC.State.Death {type=Bool}
 NPC.State.Threatened {type=Float}
 NPC.State.Unaware {type=Float}
 UE.Attenuation.Distance {type=Float}
 UE.Source.OnPlay {type=Trigger}
]
interfaces [UE.Attenuation:1.0, UE.OutputFormat.Stereo:1.0, UE.Source.OneShot:1.0, UE.Source:1.1]
outputs [
 "UE.OutputFormat.Stereo.Audio:0" {type=Audio}
 "UE.OutputFormat.Stereo.Audio:1" {type=Audio}
 UE.Source.OneShot.OnFinished {type=Trigger}
]
pages [Default {
  input_defaults {
   NPC.Event.OnMusicStop=null
   NPC.Event.OnPlayerTargeted=null
   NPC.Event.OnThreatened=null
   NPC.Music.BPM=114
   NPC.Music.CrossfadeTime=0.1
   NPC.Music.StartThreatenStingerWave=/FortNPCMusic/Audio/TempMusic/Temp_JungleBoss_AggroIntro_Stinger
   NPC.Music.StartUnawareStingerWave=null
   NPC.Music.StopStingerWave=/FortNPCMusic/Audio/TempMusic/Temp_JungleBoss_Death_Stinger
   NPC.Music.ThreatenWave=/FortNPCMusic/Audio/TempMusic/Temp_JungleBoss_Aggro_Loop
   NPC.Music.UnawareWave=/FortNPCMusic/Audio/TempMusic/Temp_JungleBoss_Idle_Loop
   NPC.State.Death=null
   NPC.State.Threatened=0.84
   NPC.State.Unaware=1
   UE.Attenuation.Distance=0
   UE.Source.OnPlay=null
  }
  nodes [
   GetWaveDuration_0 = "UE.Get Wave Duration: "(Wave=^NPC.Music.StartThreatenStingerWave)
   InitVariable.Trigger()
   InitVariable.Trigger()
   InitVariable.Trigger()
   InitVariable.Trigger()
   n1 = InitVariable.Audio()
   n2 = InitVariable.Trigger()
   n3 = InitVariable.Trigger()
   TriggerOnce_0 = UE.Trigger Once(Reset=^UE.Source.OnPlay, Trigger In=VariableDeferredAccessor.Trigger(Variable=@n3.Variable).Value)
   n4 = /FortMetasound/Patches/IsWaveAssetValid(
    WaveAsset=^NPC.Music.StopStingerWave
    Trigger In=VariableAccessor.Trigger(Variable=VariableMutator.Trigger(Variable=@n2.Variable, Value=@TriggerOnce_0.Trigger Out).Variable).Value
   )
   n5 = TriggerCompare.Clamp:Bool(A=^NPC.State.Death, Compare=@n4.Is Valid) {also {defaults {B="<MetaSoundBuilderResult.SUCCEEDED: 0>"}}}
   WavePlayer_2 = UE.Wave Player:Stereo(Wave Asset=^NPC.Music.StopStingerWave, Play=@n5.True)
   n6 = InitVariable.Audio()
   n7 = UE.BPMToSeconds(BPM=^NPC.Music.BPM)
   n8 = VariableDeferredAccessor.Trigger(Variable=@n3.Variable)
   n9 = VariableAccessor.Trigger(Variable=VariableMutator.Trigger(Variable=InitVariable.Trigger().Variable, Value=UE.TriggerRepeat(Start=^UE.Source.OnPlay, Period=@n7.Seconds, Stop=@n8.Value).RepeatOut).Variable)
   n10 = VariableAccessor.Trigger(Variable=@n9.Variable)
   n11 = VariableDeferredAccessor.Trigger(Variable=InitVariable.Trigger().Variable)
   TriggerControl_1 = UE.Trigger Control(Open=^NPC.Event.OnThreatened, Trigger In=VariableAccessor.Trigger(Variable=@n10.Variable).Value, Close=@n11.Value)
   n12 = VariableAccessor.Trigger(Variable=VariableMutator.Trigger(Value=@TriggerControl_1.Trigger Out, Variable=@n11.Variable).Variable)
   WavePlayer_4 = UE.Wave Player:Stereo(Wave Asset=^NPC.Music.StartThreatenStingerWave, Play=@n12.Value)
   n13 = InitVariable.Audio()
   n14 = InitVariable.Audio()
   TriggerAnyTriggerAny2_3 = "TriggerAny.Trigger Any (2)"(In 1=@WavePlayer_4.On Cue Point, In 0=@WavePlayer_4.On Nearly Finished)
   n15 = VariableDeferredAccessor.Trigger(Variable=InitVariable.Trigger().Variable)
   TriggerControl_0 = UE.Trigger Control(Trigger In=@TriggerAnyTriggerAny2_3.Out, Open=VariableAccessor.Trigger(Variable=@n12.Variable).Value, Close=@n15.Value)
   WavePlayer_3 = UE.Wave Player:Stereo(
    Wave Asset=^NPC.Music.ThreatenWave
    Play=VariableAccessor.Trigger(Variable=VariableMutator.Trigger(Value=@TriggerControl_0.Trigger Out, Variable=@n15.Variable).Variable).Value
    Stop=VariableDeferredAccessor.Trigger(Variable=@n3.Variable).Value
   ) {also {defaults {Loop="<MetaSoundBuilderResult.SUCCEEDED: 0>"}}}
   n16 = InitVariable.Audio()
   AudioMixerAudioMixerStereo2_1 = "AudioMixer.Audio Mixer (Stereo, 2)"(
    In 0 L=VariableAccessor.Audio(Variable=VariableMutator.Audio(Variable=@n6.Variable, Value=@WavePlayer_4.Out Left).Variable).Value
    In 0 R=VariableAccessor.Audio(Variable=VariableMutator.Audio(Variable=@n13.Variable, Value=@WavePlayer_4.Out Right).Variable).Value
    In 1 R=VariableAccessor.Audio(Variable=VariableMutator.Audio(Variable=@n14.Variable, Value=@WavePlayer_3.Out Right).Variable).Value
    In 1 L=VariableAccessor.Audio(Variable=VariableMutator.Audio(Variable=@n16.Variable, Value=@WavePlayer_3.Out Left).Variable).Value
   )
   n17 = InitVariable.Audio()
   n18 = InitVariable.Audio()
   n19 = InitVariable.Audio()
   n20 = InitVariable.Trigger()
   TriggerOnce_1 = UE.Trigger Once(Trigger In=@n10.Value, Reset=VariableDeferredAccessor.Trigger(Variable=@n3.Variable).Value)
   n21 = &IsWaveAssetValid(
    WaveAsset=^NPC.Music.StartUnawareStingerWave
    Trigger In=VariableAccessor.Trigger(Variable=VariableMutator.Trigger(Variable=@n20.Variable, Value=@TriggerOnce_1.Trigger Out).Variable).Value
   )
   WavePlayer_1 = UE.Wave Player:Stereo(Wave Asset=^NPC.Music.StartUnawareStingerWave, Play=@n21.Is Valid)
   n22 = InitVariable.Audio()
   n23 = InitVariable.Audio()
   n24 = InitVariable.Trigger()
   TriggerAnyTriggerAny2_2 = "TriggerAny.Trigger Any (2)"(In 0=@n21.Is Not Valid, In 1=@WavePlayer_1.On Finished)
   WavePlayer_0 = UE.Wave Player:Stereo(
    Wave Asset=^NPC.Music.UnawareWave
    Play=VariableAccessor.Trigger(Variable=VariableMutator.Trigger(Variable=@n24.Variable, Value=@TriggerAnyTriggerAny2_2.Out).Variable).Value
    Stop=VariableDeferredAccessor.Trigger(Variable=@n3.Variable).Value
   ) {also {defaults {Loop="<MetaSoundBuilderResult.SUCCEEDED: 0>"}}}
   n25 = InitVariable.Audio()
   AudioMixerAudioMixerStereo2_0 = "AudioMixer.Audio Mixer (Stereo, 2)"(
    In 0 R=VariableAccessor.Audio(Variable=VariableMutator.Audio(Variable=@n19.Variable, Value=@WavePlayer_1.Out Right).Variable).Value
    In 0 L=VariableAccessor.Audio(Variable=VariableMutator.Audio(Variable=@n22.Variable, Value=@WavePlayer_1.Out Left).Variable).Value
    In 1 L=VariableAccessor.Audio(Variable=VariableMutator.Audio(Variable=@n23.Variable, Value=@WavePlayer_0.Out Left).Variable).Value
    In 1 R=VariableAccessor.Audio(Variable=VariableMutator.Audio(Variable=@n25.Variable, Value=@WavePlayer_0.Out Right).Variable).Value
   )
   n26 = InitVariable.Audio()
   n27 = InitVariable.Audio()
   n28 = InitVariable.Audio()
   n29 = InitVariable.Float()
   n30 = VariableAccessor.Float(Variable=VariableMutator.Float(Variable=InitVariable.Float().Variable, Value=UE.InterpTo:Audio(Interp Time=^NPC.Music.CrossfadeTime, Target=^NPC.State.Threatened).Value).Variable)
   n31 = MapRange.MapRange:Float(In=@n30.Value) {also {defaults {Out Range A=1, Out Range B=0}}}
   AudioMixerAudioMixerStereo4_0 = "AudioMixer.Audio Mixer (Stereo, 4)"(
    In 2 L=VariableAccessor.Audio(Variable=VariableMutator.Audio(Variable=@n1.Variable, Value=@WavePlayer_2.Out Left).Variable).Value
    In 1 L=VariableAccessor.Audio(Variable=VariableMutator.Audio(Value=@AudioMixerAudioMixerStereo2_1.Out L, Variable=@n17.Variable).Variable).Value
    In 2 R=VariableAccessor.Audio(Variable=VariableMutator.Audio(Variable=@n18.Variable, Value=@WavePlayer_2.Out Right).Variable).Value
    In 0 L=VariableAccessor.Audio(Variable=VariableMutator.Audio(Value=@AudioMixerAudioMixerStereo2_0.Out L, Variable=@n26.Variable).Variable).Value
    In 0 R=VariableAccessor.Audio(Variable=VariableMutator.Audio(Value=@AudioMixerAudioMixerStereo2_0.Out R, Variable=@n27.Variable).Variable).Value
    In 1 R=VariableAccessor.Audio(Variable=VariableMutator.Audio(Value=@AudioMixerAudioMixerStereo2_1.Out R, Variable=@n28.Variable).Variable).Value
    Gain 0=VariableAccessor.Float(Variable=VariableMutator.Float(Variable=@n29.Variable, Value=@n31.Out Value).Variable).Value
    Gain 1=VariableAccessor.Float(Variable=@n30.Variable).Value
   )
   n32 = InitVariable.Float()
   n33 = InitVariable.Float()
   n34 = UE.Reroute(Value=MapRange.MapRange:Float(
     In=^UE.Attenuation.Distance
     In Range B=VariableAccessor.Float(Variable=@n32.Variable).Value
     In Range A=VariableAccessor.Float(Variable=@n33.Variable).Value
    ) {also {defaults {Out Range A=1, Out Range B=0.1}}}.Out Value)
   TriggerAnyTriggerAny2_0 = "TriggerAny.Trigger Any (2)"(In 0=@n4.Is Not Valid, In 1=@n5.False)
   ADSREnvelopeADSREnvelope_0 = ADSR Envelope.ADSR Envelope:Float(Trigger Attack=^UE.Source.OnPlay, Trigger Release=@TriggerAnyTriggerAny2_0.Out) {also {defaults {Attack Time=0, Release Time=1.5, Sustain Level=1}}}
   n35 = VariableAccessor.Float(Variable=VariableMutator.Float(Value=@ADSREnvelopeADSREnvelope_0.Out Envelope, Variable=InitVariable.Float().Variable).Variable)
   Multiply_1 = UE.Multiply:Audio by Float(
    PrimaryOperand=UE.Multiply:Audio by Float(PrimaryOperand=@AudioMixerAudioMixerStereo4_0.Out R, AdditionalOperands=@n34.Value).Out
    AdditionalOperands=VariableAccessor.Float(Variable=@n35.Variable).Value
   )
   Multiply_3 = UE.Multiply:Audio by Float(
    PrimaryOperand=UE.Multiply:Audio by Float(PrimaryOperand=@AudioMixerAudioMixerStereo4_0.Out L, AdditionalOperands=@n34.Value).Out
    AdditionalOperands=@n35.Value
   )
   TriggerAnyTriggerAny2_1 = "TriggerAny.Trigger Any (2)"(In 0=@ADSREnvelopeADSREnvelope_0.On Done, In 1=@WavePlayer_2.On Finished)
   VariableAccessorTrigger_2 = VariableAccessor.Trigger(Variable=VariableMutator.Trigger(Variable=InitVariable.Trigger().Variable, Value=@TriggerAnyTriggerAny2_1.Out).Variable)
   n36 = VariableDeferredAccessor.Bool(Variable=InitVariable.Bool().Variable)
   n37 = TriggerCompare.Clamp:Bool(Compare=^NPC.Event.OnMusicStop, A=@n36.Value)
   TriggerRouteTriggerRouteBool2_0 = "TriggerRoute.Trigger Route (Bool, 2):Bool"(Set 1=^UE.Source.OnPlay, Set 0=@n37.False) {also {defaults {Value 0="<MetaSoundBuilderResult.SUCCEEDED: 0>"}}}
   VariableMutator.Bool(Value=@TriggerRouteTriggerRouteBool2_0.Value, Variable=@n36.Variable)
   n38 = VariableDeferredAccessor.Trigger(Variable=@n3.Variable)
   TriggerControl_2 = UE.Trigger Control(Open=@n37.True, Trigger In=@n9.Value, Close=@n38.Value)
   VariableMutator.Trigger(Value=@TriggerControl_2.Trigger Out, Variable=@n38.Variable)
  ]
  links [
   {dst="@UE.OutputFormat.Stereo.Audio:1", src=Multiply_1.Out}
   {dst="@UE.OutputFormat.Stereo.Audio:0", src=Multiply_3.Out}
   {dst="@UE.Source.OneShot.OnFinished", src=VariableAccessorTrigger_2.Value}
  ]
 }]
properties {
 assetImportData=&self:AssetImportData
 bOverrideConcurrency=true
 numChannels=2
 outputFormat=Stereo
 sampleRate=48000
 soundClassObject=&Fort_Music_NPC
 concurrencyOverrides {bLimitToOwner=true, maxCount=1}
 modulationSettings {volumeRouting=Override}
 soundSubmixSends [{}]
}
