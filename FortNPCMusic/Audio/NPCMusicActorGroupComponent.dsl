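; Actor component that runs one shared, attached MetaSound music track for a group of NPCs registered through their NPC music components.
; When group members allow multiple pawns, a timer re-attaches the music to the NPC nearest the closest listener.
; Alert changes fan out through the music manager's group components, updating combat parameters on timers.
; Removing the last member triggers a stop or fade-out, then destroys the component.
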(fn AddComponent
  (params (Component :type NPCMusic%20Component%20Object%20Reference))
  (Utilities|Array|Add (Variables|Default|GetComponents) Component)
  (Variables|Default|SetActorClass (Utilities|GetClass (Components|GetOwner Component)))
  (Variables|Default|SetGroupComponentTags (Variables|/FortNPCMusic/Audio/NPCMusicComponent.NPCMusicComponent_C|GetPawnGroupTags Component))
  (CallFunction|AddTrack Component))

(fn AddTrack
  (params (Component :type NPCMusic%20Component%20Object%20Reference))
  (bind _metasound (Variables|/FortNPCMusic/Audio/NPCMusicComponent.NPCMusicComponent_C|GetMetaSound (LocalVariables|GetLocalComponent)))
  (bind _localcomponent (LocalVariables|GetLocalComponent))
  (LocalVariables|SetLocalComponent Component)
  ;[0.01938 0.01938 0.01938] Start distance check so we can update music component location, if we need too.
  (if (Variables|/FortNPCMusic/Audio/NPCMusicComponent.NPCMusicComponent_C|GetMultiplePawnsValid _localcomponent)
    (bind _returnvalue (Utilities|Time|SetTimerbyEvent (EventDispatchers|CreateEvent :!event DistanceCheck) (Variables|/FortNPCMusic/Audio/NPCMusicComponent.NPCMusicComponent_C|GetDistance%20Check%20Update _localcomponent) true))
    (Variables|Default|SetHandleDistanceCheck _returnvalue)
    (else
      _))
  ;[0.01938 0.01938 0.01938] Valid closet pawn and if music exists then play track based on outcome.
  (bind _closestactor (Variables|Default|GetClosestActor)
    (:then
      _)
    (:else
      (Variables|Default|SetClosestActor (Components|GetOwner _localcomponent))))
  (bind _music (Variables|Default|GetMusic)
    (:then
      (if (not (Variables|Default|GetbTrackValid))
        (Audio|Components|Audio|FadeOut _music (Variables|/FortNPCMusic/Audio/NPCMusicComponent.NPCMusicComponent_C|GetOverlappedMusicFadeOutTime (LocalVariables|GetLocalComponent)) 0.0 "Linear" :!class /Script/Engine.AudioComponent)))
    (:else
      _))
  (Variables|Default|SetbTrackValid true)
  (bind _returnvalue_2 (Audio|SpawnSoundAttached _metasound (Variables|/Script/Engine.Actor|GetRootComponent _closestactor) "None" "0.000000,0.000000,100.000000"))
  (bind _output_get (Variables|Default|SetMusic _returnvalue_2))
  (Class|BPFLNPCMusic|LogNPCMusic (Utilities|Text|FormatText "NSLOCTEXT(\"[DD72ED80FB0F831F2FDB512343FB39FA]\", \"6175753F49E65FD8F43B22ADB794EA36\", \"Starting music track {name}\")" (Utilities|GetDisplayName _metasound)))
  (Audio|AddEventbyComponent (GameInstanceSubsystems|GetMusicEventSubsystem) _output_get "(TagName=\"MusicEventSubsystem.Event.NPC\")" "(GameplayTags=)" 0 "")
  (Audio|Components|Audio|FadeIn _output_get (Variables|/FortNPCMusic/Audio/NPCMusicComponent.NPCMusicComponent_C|GetFadeInTime (LocalVariables|GetLocalComponent)) 1.0 0.0 "Logarithmic" :!class /Script/Engine.AudioComponent)
  (Audio|Parameter|SetFloatParameter (Variables|Default|GetMusic) (Variables|/FortNPCMusic/Audio/NPCMusicComponent.NPCMusicComponent_C|GetVolumeRadiusName (LocalVariables|GetLocalComponent)) (Variables|/FortNPCMusic/Audio/NPCMusicComponent.NPCMusicComponent_C|GetVolumeRadius (LocalVariables|GetLocalComponent)) :!class /Script/AudioExtensions.AudioParameterControllerInterface))

(fn DistanceCheck
  ;[0.01938 0.01938 0.01938] Is any controller valid
  (if (Utilities|Array|IsValidIndex (Variables|Default|GetComponents))
    ;[0.01938 0.01938 0.01938] Compare closest non player pawn to local player pawn and update when necessary.
    (for _array_element (Variables|Default|GetComponents)
      (bind (_listenerposition _returnvalue) (Audio|GetClosestListenerLocation (Transformation|GetActorLocation (Components|GetOwner _array_element)) 1e+05 true))
      (if _returnvalue
        (bind _closestactor (Variables|Default|GetClosestActor)
          (:then
            (if (<= (Math|Vector|Distance(Vector) (Transformation|GetActorLocation (Components|GetOwner _array_element)) _listenerposition) (Math|Vector|Distance(Vector) (Transformation|GetActorLocation _closestactor) _listenerposition))
              ;[0.01938 0.01938 0.01938] Update music location based on closest pawn
              (if (!= (Components|GetOwner _array_element) (Variables|Default|GetClosestActor))
                (CallFunction|SetClosestPawn (Components|GetOwner _array_element)))))
          (:else
            ;[0.01938 0.01938 0.01938] If something happens to closest pawn and is no longer valid. Set to zero index pawn so it can update location.
            (CallFunction|SetClosestPawn (Components|GetOwner (Utilities|Array|Get(acopy) (Variables|Default|GetComponents))))))))))

(fn ExecuteMusicTrigger
  (params (InName :type Name))
  (bind _music (Variables|Default|GetMusic)
    (:then
      (Audio|Parameter|ExecuteTriggerParameter _music InName))
    (:else)))

(fn GetMusicComponent
  (return (Variables|Default|GetMusic)))

(fn OnAlertStatusChanged
  (params (Component :type NPCMusic%20Component%20Object%20Reference))
  (LocalVariables|SetLocalComponent Component)
  (for _array_element (Variables|Default|GetComponents)
    (if (Utilities|Enum|Equal(Enum) (Variables|/FortNPCMusic/Audio/NPCMusicComponent.NPCMusicComponent_C|GetMusicState _array_element) "NewEnumerator2")
      (for _array_element_2 (Variables|/FortNPCMusic/Audio/NPCMusicManagerComponent.NPCMusicManagerComponent_C|GetActorGroupComponents (Variables|Default|GetNPCMusicManager))
        (Default|StartUpdate _array_element_2 true)
        (bind _music (Variables|Default|GetMusic)
          (:then
            (Audio|Parameter|ExecuteTriggerParameter _music (Variables|/FortNPCMusic/Audio/NPCMusicComponent.NPCMusicComponent_C|GetOnThreatenedTriggerName (LocalVariables|GetLocalComponent))))
          (:else)))
      (return)))
  (for _array_element_3 (Variables|/FortNPCMusic/Audio/NPCMusicManagerComponent.NPCMusicManagerComponent_C|GetActorGroupComponents (Variables|Default|GetNPCMusicManager))
    (Default|StartUpdate _array_element_3 false)))

(fn RemoveComponent
  (params (Component :type NPCMusic%20Component%20Object%20Reference))
  (bind _returnvalue (Public|GetIsNPCDead Component))
  (bind _components (Variables|Default|GetComponents))
  (bind _handledistancecheck (Variables|Default|GetHandleDistanceCheck))
  (bind _alerttimerhandle (Variables|Default|GetAlertTimerHandle))
  (Utilities|Array|RemoveItem _components Component)
  ;[0.01938 0.01938 0.01938] Stop music and clean up components.
  (if (Utilities|Array|IsEmpty _components)
    (Audio|Parameter|SetBooleanParameter (Variables|Default|GetMusic) "NPC.State.bIsDead" _returnvalue :!class /Script/AudioExtensions.AudioParameterControllerInterface)
    (if _returnvalue
      (Audio|Parameter|ExecuteTriggerParameter (Variables|Default|GetMusic) (Variables|/FortNPCMusic/Audio/NPCMusicComponent.NPCMusicComponent_C|GetOnStopTriggerName Component))
      (else
        (Audio|Components|Audio|FadeOut (Variables|Default|GetMusic) (Variables|/FortNPCMusic/Audio/NPCMusicComponent.NPCMusicComponent_C|GetFadeOutTime Component) 0.0 "Linear" :!class /Script/Engine.AudioComponent)))
    (Class|BPFLNPCMusic|LogNPCMusic "NSLOCTEXT(\"[DD72ED80FB0F831F2FDB512343FB39FA]\", \"972B3AD1443BEC0D3460CEBB2E9C11DD\", \"Stopping music\")")
    (if (Utilities|Time|IsValidTimerHandle _handledistancecheck)
      (Utilities|Time|ClearandInvalidateTimerbyHandle _handledistancecheck)
      (else
        _))
    (Variables|Default|SetbTrackValid false)
    (Variables|Default|SetbThreatTrigger false)
    (if (Utilities|Time|IsValidTimerHandle _alerttimerhandle)
      (Utilities|Time|ClearandInvalidateTimerbyHandle _alerttimerhandle)
      (else
        _))
    (Components|DestroyComponent)
    (else
      ;[0.01938 0.01938 0.01938] If non player pawn is removed from controller and was the closest but not the last, update music component to attach to a pawn that is valid.
      (if (== (Components|GetOwner Component) (Variables|Default|GetClosestActor))
        (CallFunction|SetClosestPawn (Components|GetOwner (Utilities|Array|Get(acopy) (Variables|Default|GetComponents))))))))

(fn SetClosestPawn
  (params (ClosestActor :type Actor%20Object%20Reference))
  (bind _output_get (Variables|Default|SetClosestActor ClosestActor))
  (bind _music (Variables|Default|GetMusic)
    (:then
      (Transformation|AttachComponentToComponent _music (Variables|/Script/Engine.Actor|GetRootComponent _output_get) "None" "SnapToTarget" "KeepRelative" "KeepRelative" true))
    (:else)))

(fn SetMusicBoolParam
  (params (InBool :type Boolean) (InName :type Name))
  (bind _music (Variables|Default|GetMusic)
    (:then
      (Audio|Parameter|SetBooleanParameter _music InName InBool :!class /Script/AudioExtensions.AudioParameterControllerInterface))
    (:else)))

(fn SetMusicFloatParam
  (params (FloatValue :type Float%20%28single-precision%29) (InName :type Name))
  (bind _music (Variables|Default|GetMusic)
    (:then
      (Audio|Parameter|SetFloatParameter _music InName FloatValue :!class /Script/AudioExtensions.AudioParameterControllerInterface))
    (:else)))

(fn StartUpdate
  (params (AnyOneThreatened :type Boolean))
  (Variables|Default|SetbIsInCombat AnyOneThreatened)
  (bind _returnvalue (Utilities|Time|SetTimerbyEvent (EventDispatchers|CreateEvent :!event UpdateParameters) 0.5 true))
  (Variables|Default|SetAlertTimerHandle _returnvalue))

(fn UpdateParameters
  (bind _music (Variables|Default|GetMusic)
    (:then
      (Audio|Parameter|SetBooleanParameter _music "NPC.State.bInCombat" (Variables|Default|GetbIsInCombat) :!class /Script/AudioExtensions.AudioParameterControllerInterface))
    (:else)))
