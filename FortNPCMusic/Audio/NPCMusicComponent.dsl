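; NPCMusicComponent links its owning NPC pawn or actor to the game state's NPC music manager for proximity-driven music.
; On cosmetic-capable instances it waits for playlist data, and for NPCs the ability system, then spawns an attached audio gameplay volume.
; Volume toggles and death or destruction drive manager registration.
; NPC alert changes set music state, while target changes fire spotted/lost triggers by team affiliation; the boss health handler only casts.
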(event EventBeginPlay
  (bind _returnvalue (Components|GetOwner))
  (Networking|CanExecuteCosmeticEvents
    (:True
      (bind _gamestate (Class|FortAsyncActionCurrentPlaylistReady|PlaylistDataReadyAsync)
        (:OnReady
          (Variables|Private|SetMusicManager (Actor|GetComponentbyClass _gamestate "/FortNPCMusic/Audio/NPCMusicManagerComponent.NPCMusicManagerComponent_C"))
          (bind _asbp_player_pawn_non_participant (Utilities|Casting|CastToBP_PlayerPawn_NonParticipant _returnvalue)
            (:then
              (Variables|Private|SetbIsNPC true)
              (bind _output_get (Variables|Private|SetNPCOwner _asbp_player_pawn_non_participant))
              (Utilities|IsValid (Ability|GetAbilitySystemComponent _output_get)
                (:Is%20Valid
                  (bind _output_get_2 (Variables|Private|SetNPCOwnerASC (Ability|GetAbilitySystemComponent _output_get)))
                  (Class|BPFLNPCMusic|LogNPCMusic (Utilities|Text|FormatText "NSLOCTEXT(\"[A723707B807BF98C53950608F08D1EE3]\", \"BB1128C5474C8D19E8E3CFB09AE5A1A5\", \"ASC already initialized {objectname}\")" (Utilities|GetDisplayName _output_get_2)))
                  (Public|SetupPawnBindings)
                  (CallFunction|SpawnAudioShape))
                (:Is%20Not%20Valid
                  (EventDispatchers|BindEventtoOnAbilitySystemComponentInitialized _output_get (delegate Custom|OnASCInit)))))
            (:CastFailed
              (Variables|Private|SetbIsNPC false)
              (bind _output_get_3 (Variables|Private|SetOwner _returnvalue))
              (CallFunction|SpawnAudioShape)
              (Game|BindEventtoOnDestroyed _output_get_3 (EventDispatchers|CreateEvent :!event Actor%20on%20Destroy)))))))))

(event EventEndPlay
  (params (EndPlayReason :type EEndPlayReason%20Enum :default "Destroyed"))
  (Private|RegisterActor false)
  (bind _agv (Variables|Private|GetAGV)
    (:then
      (Actor|DestroyActor _agv))
    (:else)))

(event Custom|OnASCInit
  (params (AbilitySystemComponent :type Fort%20Ability%20System%20Component%20Object%20Reference) (AffectedPawn :type Fort%20Player%20Pawn%20Object%20Reference))
  (bind _output_get (Variables|Private|SetNPCOwnerASC AbilitySystemComponent))
  (Class|BPFLNPCMusic|LogNPCMusic (Utilities|Text|FormatText "NSLOCTEXT(\"[A723707B807BF98C53950608F08D1EE3]\", \"F5FD67F74F3F3248608C5F976C957675\", \"ASC Initialized {objectname}\")" (Utilities|GetDisplayName _output_get)))
  (Public|SetupPawnBindings)
  (CallFunction|SpawnAudioShape))

(event Custom|SpawnAudioShape
  (bind _actorowner (Public|GetActorBase))
  ;[1 1 1] Spawn Controller and attached to actor.
  (bind _returnvalue (Game|SpawnActorfromClass (Variables|TriggerVolume|GetAGVClass) (Transformation|GetActorTransform _actorowner) "Undefined" "SelectDefaultAtRuntime" 0 0))
  (bind _output_get (Variables|Private|SetAGV _returnvalue))
  (Transformation|AttachActorToActor _output_get _actorowner "None" "SnapToTarget" "SnapToTarget" "KeepRelative" true)
  (Private|AGV|ConfigureAGV))

(fn ActoronDestroy
  (params (DestroyedActor :type Actor%20Object%20Reference))
  (Variables|Private|SetbActorDead true)
  (Private|RegisterActor false))

(fn ConfigureAGV
  (bind _volumeradius (Variables|Volume|GetVolumeRadius))
  (bind _agv (Variables|Private|GetAGV))
  (bind _audiogameplayvolume (Variables|/FortNPCMusic/Audio/AGV_NPCMusicVolume.AGV_NPCMusicVolume_C|GetAudioGameplayVolume _agv))
  (Control|SetRadius (Variables|/FortNPCMusic/Audio/AGV_NPCMusicVolume.AGV_NPCMusicVolume_C|GetAudioShapeSphere _agv) _volumeradius :!class /Script/AudioShapes.AudioShapeSphereComponent)
  (Events|BindEventtoOnToggledOn _audiogameplayvolume (EventDispatchers|CreateEvent :!event OnPlayerEnterAGV))
  (Events|BindEventtoOnToggledOff _audiogameplayvolume (EventDispatchers|CreateEvent :!event OnPlayerExitAGV))
  (Class|BPFLNPCMusic|LogNPCMusic (Utilities|Text|FormatText "NSLOCTEXT(\"[A723707B807BF98C53950608F08D1EE3]\", \"4421759C4B66DB73D3A6879CA1B6030A\", \"AGV {name} created with radius {radius}\")" (Utilities|GetDisplayName _agv) _volumeradius)))

(fn GetIsNPCDead
  (if (Variables|Private|GetbIsNPC)
    (return (FortPawn|IsDead (Variables|Private|GetNPCOwner)))
    (else
      (return (Variables|Private|GetbActorDead)))))

(fn GetMusicManager
  (return (Variables|Private|GetMusicManager)))

(fn GetActorBase
  (return (select (Variables|Private|GetbIsNPC) (Variables|Private|GetNPCOwner) (Variables|Private|GetOwner))))

(fn GetAGV
  (bind _agv (Variables|Private|GetAGV)
    (:then
      (return _agv true))
    (:else
      (return 0 false))))

(fn GetNPCOwner
  (return (Variables|Private|GetNPCOwner)))

(fn OnAlertLevelChanged
  (params (OldAlertLevel :type EAlertLevel%20Enum :default "Unaware") (NewAlertLevel :type EAlertLevel%20Enum :default "Unaware"))
  (Variables|Private|SetMusicState (Utilities|Select :Index NewAlertLevel :Unaware "NewEnumerator0" :Suspicious "NewEnumerator0" :LostTarget "NewEnumerator2" :Threatened "NewEnumerator2" :Count "NewEnumerator0"))
  (Class|NPCMusicManagerComponent|OnAlertLevelChange (Variables|Private|GetMusicManager) self))

(fn OnCurrentTargetedActor
  (params (NewTarget :type Actor%20Object%20Reference))
  (bind _self self)
  (bind _musicmanager (Variables|Private|GetMusicManager))
  (bind _asfort_pawn (Utilities|Casting|CastToFortPawn NewTarget))
  (if (Utilities|Enum|Equal(Enum) (Team|GetTeamAffiliationforActortoLocalPlayer _asfort_pawn) "Friendly")
    (Class|NPCMusicManagerComponent|ExecuteMusicTrigger _musicmanager _self (Variables|MetaSoundParams|GetOnPlayerSpottedTriggerName))
    (else
      (Class|NPCMusicManagerComponent|ExecuteMusicTrigger _musicmanager _self (Variables|MetaSoundParams|GetOnPlayerLostTriggerName)))))

(fn OnDeath
  (params (Cause :type EDeathCause%20Enum :default "OutsideSafeZone"))
  (Private|RegisterActor false))

(fn OnHealthChange
  (bind _asbp_player_pawn_non_participant_boss (Utilities|Casting|CastToBP_PlayerPawn_NonParticipant_Boss (Variables|Private|GetNPCOwner))))

(fn OnPlayerEnterAGV
  (Private|RegisterActor true)
  (Class|BPFLNPCMusic|LogNPCMusic "NSLOCTEXT(\"[A723707B807BF98C53950608F08D1EE3]\", \"122A0B50418538E082CB06BD21505FE4\", \"Entering AGV radius\")")
  (bind _npcowner (Variables|Private|GetNPCOwner)
    (:then
      (Private|Delegates|OnAlertLevelChanged (Variables|/Game/Athena/AI/NPCs/Base/BP_PlayerPawn_NonParticipant.BP_PlayerPawn_NonParticipant_C|GetOldAlertLevel _npcowner) (Variables|/Game/Athena/AI/NPCs/Base/BP_PlayerPawn_NonParticipant.BP_PlayerPawn_NonParticipant_C|GetAlertLevel _npcowner)))
    (:else)))

(fn OnPlayerExitAGV
  (Private|RegisterActor false)
  (Class|BPFLNPCMusic|LogNPCMusic "NSLOCTEXT(\"[A723707B807BF98C53950608F08D1EE3]\", \"3AE342374D9A7F5AB2BC87B84E418DB0\", \"Exiting AGV radius\")"))

(fn RegisterActor
  (params (Registered :type Boolean))
  (bind _self self)
  (if (!= Registered (Variables|Private|GetbIsRegistered))
    (bind _musicmanager (Variables|Private|GetMusicManager)
      (:then
        (bind _output_get (Variables|Private|SetbIsRegistered Registered))
        (if _output_get
          (Register|RegisterComponent _musicmanager _self)
          (else
            (Register|UnregisterComponent _musicmanager _self)))
        (Class|BPFLNPCMusic|LogNPCMusic (Utilities|Text|FormatText "NSLOCTEXT(\"[A723707B807BF98C53950608F08D1EE3]\", \"F6CD254E4D265548B37AC6B31D6D14DC\", \"Register [{bool}] with Manager {name}\")" _output_get (Utilities|GetDisplayName _musicmanager))))
      (:else))))

(fn SetupPawnBindings
  (Utilities|FlowControl|Sequence
    (:then_0
      ;[1 1 1] NPC pawn bindings
      (bind _npcowner (Variables|Private|GetNPCOwner)
        (:then
          (Default|BindEventtoOnDeath _npcowner (EventDispatchers|CreateEvent :!event Ondeath))
          (Default|BindEventtoOnTargetChanged _npcowner (EventDispatchers|CreateEvent :!event OnCurrentTargetedActor))
          (Default|BindEventtoOnAlertLevelChanged _npcowner (EventDispatchers|CreateEvent :!event OnAlertLevelChanged)))
        (:else)))
    (:then_1
      (bind _npcowner_2 (Variables|Private|GetNPCOwner)
        (:then
          (bind _asbp_player_pawn_non_participant_boss (Utilities|Casting|CastToBP_PlayerPawn_NonParticipant_Boss _npcowner_2))
          (Default|BindEventtoOnHealthChange _asbp_player_pawn_non_participant_boss (EventDispatchers|CreateEvent :!event OnHealthChange)))
        (:else)))))
