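; NPC music component for VividRazor vaults that drives staged music and boss cues through the NPC Music Manager.
; Entering the AGV radius registers the actor, sets its music to the diegetic category, groups it by the first pawn group tag and starts the owner's current stage.
; Stage enums select loop, transition or ending triggers.
; Boss functions set presence or upcoming-spawn bools; appearance and defeat also fire triggers.
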(event EventBeginPlay
  (|CallParentFunction) ; This node is disabled and will not be called.\nDrag off pins to build functionality.
)

(event EventTick
  (params (DeltaSeconds :type Float%20%28single-precision%29)))

(event Custom|Boss%20Spawned
  (bind _musicmanager (Variables|Private|GetMusicManager)
    (:then
      (Class|NPCMusicManagerComponent|SetMusicBoolParam _musicmanager self true "bIsBossPresent")
      (Class|NPCMusicManagerComponent|ExecuteMusicTrigger (Variables|Private|GetMusicManager) self "NPC.Event.BossAppeared"))
    (:else)))

(event Custom|Boss%20Defeated
  (bind _musicmanager (Variables|Private|GetMusicManager)
    (:then
      (Class|NPCMusicManagerComponent|SetMusicBoolParam _musicmanager self false "bIsBossPresent")
      (Class|NPCMusicManagerComponent|ExecuteMusicTrigger (Variables|Private|GetMusicManager) self "NPC.Event.BossEliminated")
      (Development|PrintText (Utilities|Text|FormatText "NSLOCTEXT(\"[E1985D0406032E972EDC8240775694C4]\", \"5A90ED5044C3194F321EC5A2B551E8D2\", \"Boss defeated!\")") true true "(R=0.726466,G=1.000000,B=0.000000,A=1.000000)" 4.0 "VividRazorMusicBossDefeated"))
    (:else)))

(fn OnPlayerEnterAGV
  (bind _pawngrouptags (Variables|Pawn|GetPawnGroupTags))
  (bind _gameplaytags (GameplayTags|BreakGameplayTagContainer _pawngrouptags))
  (bind _self self)
  (bind _musicmanager (CallFunction|GetMusicManager))
  (bind (_isvalid _groupcomponent) (Class|NPCMusicManagerComponent|GetGroupComponent _musicmanager _self))
  (bind _returnvalue (GameInstanceSubsystems|GetMusicEventSubsystem))
  (if (Variables|Default|GetEnable%20Music)
    (Private|RegisterActor true)
    (Class|BPFLNPCMusic|LogNPCMusic "NSLOCTEXT(\"[E1985D0406032E972EDC8240775694C4]\", \"73C6D4044875DB36B77DD48C1041FB58\", \"Entering AGV radius\")")
    (Utilities|FlowControl|Sequence
      (:then_0
        (Utilities|IsValid _musicmanager
          (:Is%20Valid
            (if _isvalid
              (bind _music (Variables|/FortNPCMusic/Audio/NPCMusicActorGroupComponent.NPCMusicActorGroupComponent_C|GetMusic _groupcomponent)
                (:then
                  (Audio|ChangeEventCategory _returnvalue _music "(TagName=\"MusicEventSubsystem.Event.NPC.Diegetic\")")
                  (if (Utilities|Array|IsNotEmpty _gameplaytags)
                    (Audio|ChangeEventGroup _returnvalue _music (Utilities|String|StringToName (GameplayTags|GetDebugStringfromGameplayTag (Utilities|Array|Get(aref) _gameplaytags))))))
                (:else))))))
      (:then_1
        (Utilities|FlowControl|Sequence
          (:then_0
            (bind _asnpcmusic_interface (Utilities|Casting|CastToNPCMusicInterface (Components|GetOwner)))
            (bind _music_stage (Music|GetMusicStage _asnpcmusic_interface))
            (Default|SetMusicStage _music_stage)))))))

(fn SetIsBossGoingtoSpawn
  (params (bIsBossGoingToSpawn :type Boolean))
  (Class|NPCMusicManagerComponent|SetMusicBoolParam (Variables|Private|GetMusicManager) self bIsBossGoingToSpawn "bIsBossGoingToSpawn")
  (Development|PrintText (Utilities|Text|FormatText "NSLOCTEXT(\"[E1985D0406032E972EDC8240775694C4]\", \"C8E1E3AA44D9BB2328D71A89408FB6F7\", \"VividRazorVaultMusic: Is Boss Going to Spawn: {bool}\")" bIsBossGoingToSpawn) true true "(R=1.000000,G=0.000000,B=0.000000,A=1.000000)" 2.0 "MusicBossAppear"))

(fn SetMusicStage
  (params (Music%20Stage :type E_NPCMusic_Stage%20Enum :default "NewEnumerator0"))
  (if (Variables|Default|GetEnable%20Music)
    (Class|NPCMusicManagerComponent|ExecuteMusicTrigger (Variables|Private|GetMusicManager) self (Utilities|Select :Index Music%20Stage :NewEnumerator0 "NPC.Event.PlayStage1Loop" :NewEnumerator1 "NPC.Event.PlayStage2Loop" :NewEnumerator2 "NPC.Event.PlayStage3Loop"))))

(fn TransitiontoNextMusicStage
  (params (Next%20Music%20Stage :type E_NPCMusic_Stage%20Enum :default "NewEnumerator0"))
  (if (Variables|Default|GetEnable%20Music)
    (Class|NPCMusicManagerComponent|ExecuteMusicTrigger (Variables|Private|GetMusicManager) self (Utilities|Select :Index Next%20Music%20Stage :NewEnumerator0 "NPC.Event.PlayTransitionToStage1" :NewEnumerator1 "NPC.Event.PlayTransitionToStage2" :NewEnumerator2 "NPC.Event.PlayTransitionToStage3" :NewEnumerator3 "NPC.Event.PlayEnding"))
    (Development|PrintText (Utilities|Text|FormatText "NSLOCTEXT(\"[E1985D0406032E972EDC8240775694C4]\", \"F772EE6B473B67649D2231AD018F0D70\", \"Transitioning music stage: {enum}\")" (Utilities|String|EnumtoString Next%20Music%20Stage)) true true "(R=0.000000,G=1.000000,B=0.448548,A=1.000000)" 4.0 "VividRazorTransitionMusic")))
