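; NPCMusicInterface is an interface, named for NPC music, that exposes a GetMusicStage query for the current music stage.
; Its single function returns an enumerator value, shown only at the first entry of its enum, with no other logic in the file.
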(fn GetMusicStage
  (return "NewEnumerator0"))
