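; NPC music manager component that sorts NPC music components into actor group components and relays music events to them.
; Registration looks up a group by owner class or pawn group tags, adding a new group component to its owner when none is found.
; Music triggers, bool/float parameters and alert changes are forwarded to that group.
; Group validity checks drive default submix registration with the audio analysis subsystem, released on EndPlay.
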(event EventEndPlay
  (params (EndPlayReason :type EEndPlayReason%20Enum :default "Destroyed"))
  (if (Variables|Default|GetisSubmixRegistered)
    (AudioAnalysis|UnRegisterDefaultSubmix (WorldSubsystems|GetFortAudioAnalysisSubsystem))))

(fn CheckForRegisteredSubmix
  (bind _actorgroupcomponents (Variables|Default|GetActorGroupComponents))
  (bind _returnvalue (Utilities|Array|IsNotEmpty _actorgroupcomponents))
  (bind _returnvalue_2 (WorldSubsystems|GetFortAudioAnalysisSubsystem))
  (if _returnvalue
    (bind _output_get (LocalVariables|SetLocalComponents (Variables|Default|GetActorGroupComponents)))
    (bind _array_element (Utilities|Array|ForEachLoopwithBreak _output_get)
      (:LoopBody
        (Utilities|IsValid _array_element
          (:Is%20Valid
            (LocalVariables|SetContainsValidComponent true))
          (:Is%20Not%20Valid
            _))
        (Utilities|Array|RemoveItem (Variables|Default|GetActorGroupComponents) _array_element))
      (:Completed
        _))
    (if (not (== (LocalVariables|GetContainsValidComponent) (Variables|Default|GetisSubmixRegistered)))
      (Variables|Default|SetisSubmixRegistered (LocalVariables|GetContainsValidComponent)))
    (elif (== _returnvalue (Variables|Default|GetisSubmixRegistered))
      _
      (else
        (Variables|Default|SetisSubmixRegistered (Utilities|Array|IsNotEmpty (Variables|Default|GetActorGroupComponents))))))
  (if (Variables|Default|GetisSubmixRegistered)
    (AudioAnalysis|RegisterDefaultSubmix _returnvalue_2)
    (else
      (AudioAnalysis|UnRegisterDefaultSubmix _returnvalue_2))))

(fn ExecuteMusicTrigger
  (params (Component :type NPCMusic%20Component%20Object%20Reference) (InName :type Name))
  (bind (_isvalid _groupcomponent) (CallFunction|GetGroupComponent Component))
  (if _isvalid
    (Class|NPCMusicActorGroupComponent|ExecuteMusicTrigger _groupcomponent InName)))

(fn GetGroupComponent
  (params (Component :type NPCMusic%20Component%20Object%20Reference%20%28by%20ref%29))
  (for _array_element (Variables|Default|GetActorGroupComponents)
    (Utilities|IsValid _array_element
      (:Is%20Valid
        (if (== (Variables|/FortNPCMusic/Audio/NPCMusicActorGroupComponent.NPCMusicActorGroupComponent_C|GetActorClass _array_element) (Utilities|GetClass (Components|GetOwner Component)))
          _
          (elif (== (Variables|/FortNPCMusic/Audio/NPCMusicActorGroupComponent.NPCMusicActorGroupComponent_C|GetGroupComponentTags _array_element) (Variables|/FortNPCMusic/Audio/NPCMusicComponent.NPCMusicComponent_C|GetPawnGroupTags Component))
            _))
        (return true _array_element))))
  (return false 0))

(fn OnAlertLevelChange
  (params (Component :type NPCMusic%20Component%20Object%20Reference))
  (bind (_isvalid _groupcomponent) (CallFunction|GetGroupComponent Component))
  (if _isvalid
    (Default|OnAlertStatusChanged _groupcomponent Component)))

(fn RegisterComponent
  (params (Component :type NPCMusic%20Component%20Object%20Reference))
  (bind (_isvalid _groupcomponent) (CallFunction|GetGroupComponent Component))
  (if _isvalid
    (Class|NPCMusicActorGroupComponent|AddComponent _groupcomponent Component)
    (else
      (bind _returnvalue (Game|AddComponentbyClass (Components|GetOwner) "/FortNPCMusic/Audio/NPCMusicActorGroupComponent.NPCMusicActorGroupComponent_C"))
      (Variables|/FortNPCMusic/Audio/NPCMusicActorGroupComponent.NPCMusicActorGroupComponent_C|SetNPCMusicManager self _returnvalue)
      (Utilities|Array|Add (Variables|Default|GetActorGroupComponents) _returnvalue)
      (Class|NPCMusicActorGroupComponent|AddComponent _returnvalue Component)
      (CallFunction|CheckforRegisteredSubmix))))

(fn SetMusicBoolParam
  (params (Component :type NPCMusic%20Component%20Object%20Reference%20%28by%20ref%29) (InBool :type Boolean) (InName :type Name))
  (bind (_isvalid _groupcomponent) (CallFunction|GetGroupComponent Component))
  (if _isvalid
    (Class|NPCMusicActorGroupComponent|SetMusicBoolParam _groupcomponent InBool InName)))

(fn SetMusicFloatParam
  (params (Component :type NPCMusic%20Component%20Object%20Reference%20%28by%20ref%29) (InName :type Name) (FloatValue :type Float%20%28single-precision%29))
  (bind (_isvalid _groupcomponent) (CallFunction|GetGroupComponent Component))
  (if _isvalid
    (Class|NPCMusicActorGroupComponent|SetMusicFloatParam _groupcomponent FloatValue InName)))

(fn UnregisterComponent
  (params (Component :type NPCMusic%20Component%20Object%20Reference))
  (bind (_isvalid _groupcomponent) (CallFunction|GetGroupComponent Component))
  (if _isvalid
    (Class|NPCMusicActorGroupComponent|RemoveComponent _groupcomponent Component)
    (CallFunction|CheckforRegisteredSubmix)))
